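# Physics asset for the ThornedEden axe, a cosmetic pickaxe melee weapon, providing its simplified collision shape.
# A single body on the Melee_TwoHanded_Handle bone holds one capsule, offset along that bone.
# It has no constraints, so it defines no jointed or secondary-motion parts.
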PhysicsAsset /BRCosmetics/Weapons/FORT_Melee/Pickaxe_ThornedEden/Meshes/ThornedEden_Axe_PhysicsAsset
bodies {Melee_TwoHanded_Handle [{shapeType=Capsule, length=39.6374, radius=12.1002, center=<-0.5436 0.0699 19.4819>, rotation=<rot 0.6353 -90 0.0636>}]}
constraints []
